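# Material function for the DragonCart orb barrier that computes an expanding ripple from a given position, radius and normalized time.
# A soft world-space sphere mask grows with radius times time, and a clamped sine of it carves a ring.
# A time-lerped exponent tightens the ring as it expands.
# It outputs the ring, a time-faded sphere, and a faded UV distortion built from the tangent-space offset from the ripple center.
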MaterialFunction /DragonCartNarrative/Assets/VFX/OrbBarrier/MF_DragonCart_Orb_Barrier_Ripple
graph {nodes [
  n1 = WorldPosition()
  n2 = FunctionInput() {inputName=RipplePosition, sortPriority=10}
  n3 = FunctionInput() {inputName=RippleRadius, inputType=FunctionInput_Scalar, sortPriority=20}
  n4 = FunctionInput() {inputName=RippleTime, inputType=FunctionInput_Scalar, sortPriority=30}
  n5 = SphereMask(A=@n1.XYZ, B=@n2, Radius=Multiply(A=@n3, B=@n4)) {hardnessPercent=0}
  n6 = Sine(None=@n5)
  n7 = Power(Base=Multiply(A=@n5, B=Saturate(None=@n6)), Exp=LinearInterpolate(Alpha=@n4) {constA=0.6, constB=7}) {constExponent=3}
  FunctionOutput(None=@n7) {outputName=Ripple}
  n8 = OneMinus(None=@n4)
  FunctionOutput(None=Multiply(A=@n5, B=@n8)) {outputName=Sphere, sortPriority=1}
  n9 = Subtract(A=@n1.XYZ, B=@n2)
  n10 = Divide(
   A=ComponentMask(None=Transform(None=@n9) {transformSourceType=TRANSFORMSOURCE_World, transformType=TRANSFORM_Tangent}) {g=true, r=true}
   B=@n3
  )
  FunctionOutput(None=Multiply(A=Multiply(A=@n10, B=@n7), B=@n8)) {outputName=UVDistortion, sortPriority=2}
 ]}
