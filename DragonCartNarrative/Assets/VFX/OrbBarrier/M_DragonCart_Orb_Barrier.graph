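# Translucent, unlit, two-sided material for the DragonCart orb barrier, a glowing energy shell with cube-style noise, runes and sparkles.
# World-aligned noise and an analytic camera-ray sphere drive a height-blended inner glow, while bump offsets give parallax to runes and sparkles.
# Three timed, positioned ripple slots supply glow and UV distortion, alongside player-proximity masks and depth-fade intersection glow.
# Distance-based offset logic scales vertices toward the actor, and quality switches simplify shading on Low.
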Material /DragonCartNarrative/Assets/VFX/OrbBarrier/M_DragonCart_Orb_Barrier
graph {
 properties {
  bAutomaticallySetUsageInEditor=false
  bCastRayTracedShadows=false
  blendMode=BLEND_Translucent
  pixelDepthOffsetMode=PDOM_Legacy
  shadingModel=MSM_Unlit
  twoSided=true
  usedShadingModels=Unlit
 }
 nodes [
  n1 = WorldPosition()
  n2 = NamedRerouteUsage()
  n3 = SphereMask(A=@n1.XYZ, B=@n2, Radius=ScalarParameter() {defaultValue=10000, parameterName=NearPlayerEmissiveRadius}) {hardnessPercent=0.2}
  n4 = Add(A=NamedRerouteUsage(), B=NamedRerouteUsage())
  Add(A=@n3, B=@n4)
  n5 = Power(Base=NamedRerouteUsage(), Exp=Constant() {r=3})
  Add(A=@n5, B=ScalarParameter() {defaultValue=-0.05, group=Noise, parameterName=Blurred Shadow Add})
  Constant() {r=1}
  n6 = Time()
  n7 = Constant() {r=1}
  Fmod(A=@n6, B=@n7)
  Fmod(A=Multiply(A=@n6) {constB=0.73}, B=@n7)
  Fmod(A=Multiply(A=@n6) {constB=1.12}, B=@n7)
  n8 = VectorParameter() {group=Runes, parameterName=GlowColor1, defaultValue=<rgba 0.012 0.0011 0.08 1>}
  n9 = VectorParameter() {group=Runes, parameterName=GlowColor2, defaultValue=<rgba 0.0356 0.001 0.2 1>}
  n10 = VectorParameter() {group=Runes, parameterName=GlowColor3, defaultValue=<rgba 0.1264 0 0.3 1>}
  n11 = VectorParameter() {group=Runes, parameterName=GlowColor4, defaultValue=<rgba 0.3583 0 0.5 1>}
  n12 = Multiply(A=@n11.RGB, B=@n11.RGB)
  n13 = Saturate(None=Add(A=NamedRerouteUsage(), B=ScalarParameter() {defaultValue=0.25, group=Noise, parameterName=BaseCubeBackground}))
  n14 = ScalarParameter() {defaultValue=2.75, group=Noise, parameterName=Blurred Noise Max Sine}
  n15 = ScalarParameter() {defaultValue=1.5, group=Noise, parameterName=Blurred Noise Min Sine}
  n16 = NamedRerouteUsage()
  n17 = Multiply(A=@n13, B=Power(Base=NamedRerouteUsage(), Exp=LinearInterpolate(A=@n14, B=@n15, Alpha=@n16)))
  n18 = Multiply(
   A=LinearInterpolate(A=Constant() {r=1200}, B=Constant() {r=1400}, Alpha=@n16)
   B=ScalarParameter() {defaultValue=1, group=Sphere Mask, parameterName=RadiusMult}
  )
  n19 = Multiply(A=@n18, B=@n18) {desc="r^2"}
  n20 = ObjectPositionWS()
  n21 = Add(A=VectorParameter() {group=Sphere Mask, parameterName=Object Position Offset, defaultValue=<rgba 0 0 -400 1>}.RGB, B=@n20)
  n22 = NamedRerouteUsage()
  n23 = Subtract(
   A=CameraPositionWS()
   B=Add(
    A=VectorParameter() {desc=For zapping players, group=Sphere Mask, parameterName=SphereOffset}.RGB
    B=LinearInterpolate(A=@n20, B=@n21, Alpha=@n22)
   )
  )
  n24 = Distance(A=@n23, B=Constant())
  n25 = DotProduct(A=@n23, B=CameraVectorWS())
  n26 = Subtract(A=Multiply(A=@n24, B=@n24), B=Multiply(A=@n25, B=@n25))
  n27 = SquareRoot(None=Max(A=Subtract(A=@n19, B=@n26)) {constB=0})
  n28 = Multiply(A=Divide(A=Multiply(A=@n27) {constB=2}, B=Multiply(A=@n18) {constB=2}), B=ScalarParameter() {defaultValue=0.87, parameterName=gr334})
  n29 = MaterialFunctionCall(Transition Phase=@n28, Height Texture=Multiply(A=NamedRerouteUsage(), B=QualitySwitch(Default=@n17, Low=@n13, High=@n17))) {materialFunction=&HeightLerp}
  n30 = Multiply(
   A=Power(Base=@n28, Exp=ScalarParameter() {defaultValue=20, parameterName=gpow3})
   B=ScalarParameter() {defaultValue=75, parameterName=ActiveHotspot}
  )
  n31 = Power(Base=NamedRerouteUsage(), Exp=Constant() {r=2})
  n32 = Multiply(A=@n30, B=@n31)
  n33 = LinearInterpolate(
   A=@n30
   B=QualitySwitch(Default=@n32, Low=@n30, High=@n32)
   Alpha=ScalarParameter() {defaultValue=1, group=Cube_OG, parameterName=Blend-Tex-Sphere}
  )
  n34 = LinearInterpolate(A=Multiply(A=NamedRerouteUsage(), B=@n33), B=@n33, Alpha=ScalarParameter() {group=Cube_OG, parameterName=RemoveEdgeFromInnerGlow})
  n35 = Multiply(
   A=Add(
    A=MaterialFunctionCall(1=@n8.RGB, 2=@n9.RGB, 3=@n10.RGB, 4=@n12, A=@n29.Lerp Alpha No Contrast) {materialFunction=&Lerp_Multiple_Float3}.Lerp 4 Inputs
    B=Multiply(A=@n12, B=@n34)
   )
   B=NamedRerouteUsage()
  )
  n36 = ScalarParameter() {defaultValue=0.25, group=Cube_OG, parameterName=deactivated_1}
  n37 = ScalarParameter() {defaultValue=0.05, group=Cube_OG, parameterName=activated_1}
  n38 = NamedRerouteUsage()
  n39 = Saturate(None=Add(A=LinearInterpolate(A=@n36, B=@n37, Alpha=@n38), B=@n31))
  n40 = Multiply(A=@n35, B=@n39)
  LinearInterpolate(B=@n40, Alpha=NamedRerouteUsage())
  n41 = ScalarParameter() {defaultValue=2, group=Noise, parameterName=Master Noise Scale}
  n42 = TextureCoordinate()
  n43 = NamedRerouteUsage()
  n44 = Add(A=Multiply(A=@n41, B=@n42), B=@n43)
  n45 = Constant() {r=0.58}
  n46 = Constant() {r=-0.5}
  n47 = NamedRerouteUsage()
  LinearInterpolate(
   B=TextureSample!(
    UVs=BumpOffset(Coordinate=@n44, Height=@n45, HeightRatioInput=@n46)
    Tex=@n47
    Bias=ScalarParameter() {defaultValue=4, group=Cube_OG, parameterName=Mipbias_Shadow}
    Apply View MipBias
   ) {mipValueMode=TMVM_MipBias, samplerType=SAMPLERTYPE_Masks, texture=/Game/Global/Textures/Default/Noise/T_CubeNoise}.G
   Alpha=NamedRerouteUsage()
  )
  n48 = NamedRerouteUsage()
  n49 = TextureSample!(UVs=Add(A=Multiply(A=@n41, B=@n42), B=@n43), Tex=@n48, Apply View MipBias) {samplerSource=SSM_Wrap_WorldGroupSettings, samplerType=SAMPLERTYPE_Masks, texture=&T_CubeNoise}
  LinearInterpolate(B=@n49.R, Alpha=NamedRerouteUsage())
  n50 = VectorParameter() {parameterName=RippleColor, defaultValue=<rgba 3 0.0306 1.4991 1>}
  n51 = Add(A=Multiply(A=Multiply(A=@n50.RGB) {constB=2}, B=NamedRerouteUsage()), B=Multiply(A=NamedRerouteUsage()) {constB=0.5})
  n52 = ScalarParameter() {defaultValue=3, parameterName=HDRInensity}
  n53 = MaterialFunctionCall(LDR Input=Clamp(None=@n51), HDR Tint=@n51, HDR Intensity Multiplier=@n52, Desaturate Luminance=Constant() {r=0.1}) {materialFunction=&DeriveHDRfromLDR}
  n54 = Multiply(A=Multiply(A=@n51, B=@n52)) {constB=2}
  n55 = ScalarParameter() {defaultValue=1, parameterName=NearMask Emissvie Str}
  n56 = Multiply(A=NamedRerouteUsage(), B=VectorParameter() {parameterName=RippleColor, defaultValue=<rgba 3 0.0306 1.4991 1>}.RGB)
  MaterialFunctionCall(Emissive=Add(A=QualitySwitch(Default=@n53.RemappedRange, Low=@n54), B=Multiply(A=@n55, B=@n56))) {materialFunction=&MF_FX_EmissiveAdjustments}
  n57 = Power(
   Base=SphereMask(
    A=WorldPosition().XYZ
    B=VectorParameter() {group=Ripple Anim, parameterName=Ripple Position, defaultValue=<rgba -8743.291 -2119.8438 5633.1729 1>}.RGB
    Radius=ScalarParameter() {defaultValue=2000, parameterName=RippleRadius}
   ) {hardnessPercent=0.5}
   Exp=ScalarParameter() {defaultValue=2, group=Ripple Anim, parameterName=Ripple Power}
  )
  n58 = ScalarParameter() {defaultValue=1, group=Ripple Anim, parameterName=Ripple Fade}
  n59 = Sine(None=Saturate(None=Multiply(A=@n57, B=@n58)))
  n60 = LinearInterpolate(Alpha=Saturate(None=Multiply(A=@n57, B=@n59)))
  Multiply(A=@n60, B=NamedRerouteUsage())
  n61 = Multiply(A=NamedRerouteUsage(), B=ScalarParameter() {defaultValue=0.25, parameterName=IntersectGlowOpacity}) {constB=0.5}
  n62 = Multiply(A=ScalarParameter() {defaultValue=1, parameterName=OverallOpacity}, B=NamedRerouteUsage())
  Multiply(A=NamedRerouteUsage(), B=QualitySwitch(Default=Max(A=@n61, B=@n62), Low=@n62)) {constA=1}
  n63 = ScalarParameter() {defaultValue=1, group=WPO, parameterName=Scale Near}
  n64 = ScalarParameter() {group=WPO, parameterName=Scale Far}
  n65 = MaterialFunctionCall(Distance=ScalarParameter() {defaultValue=5000, group=WPO, parameterName=NearFade, sortPriority=16}) {materialFunction=&Near_Camera_Fade}
  Multiply(A=Subtract(A=ObjectPositionWS(), B=NamedRerouteUsage()), B=Subtract(A=LinearInterpolate(A=@n63, B=@n64, Alpha=@n65.Result)))
  n66 = OneMinus(None=Fresnel())
  Multiply(A=NamedRerouteUsage(), B=@n66)
  n67 = ScalarParameter() {defaultValue=1, parameterName=01_OpOverride}
  Multiply(B=@n67)
  Multiply(B=@n67)
  n68 = CollectionParameter() {collection=&FortniteMaterialParameters, parameterName=PlayerPosition}
  n69 = ComponentMask(None=@n68) {b=true, g=true, r=true}
  n70 = CameraPositionWS()
  NamedRerouteDeclaration(None=If(A=Length(None=@n69) {is=class MaterialExpressionLength}, "A > B"=@n69, "A == B"=@n70, "A < B"=@n70) {equalsThreshold=1}) {name=Player/CameraPosition, nodeColor=<rgba 1 0 0.7734 1>}
  NamedRerouteDeclaration(None=Multiply(A=@n60, B=ScalarParameter() {defaultValue=1, group=Ripple Anim, parameterName=Ripple Str})) {name=Impact Wave, nodeColor=<rgba 0 1 0.1562 1>}
  n71 = DepthFade(FadeDistance=ScalarParameter() {defaultValue=100, parameterName=IntersectGlowSize})
  NamedRerouteDeclaration(None=Power(Base=OneMinus(None=@n71), Exp=ScalarParameter() {defaultValue=5, parameterName=IntersectGlowPower}) {constExponent=3}) {name=IntersectGlow, nodeColor=<rgba 0.0312 1 0 1>}
  n72 = SphereMask(A=@n1.XYZ, B=@n2, Radius=ScalarParameter() {defaultValue=10000, parameterName=NearPlayerRadius}) {hardnessPercent=10}
  NamedRerouteDeclaration(None=Saturate(None=Add(A=@n4, B=@n72))) {name=NearMask, nodeColor=<rgba 0 0.3437 1 1>}
  n73 = ScalarParameter() {defaultValue=1, group=Runes, parameterName=Rune Masked MinGrid}
  n74 = ScalarParameter() {defaultValue=-1, group=Runes, parameterName=Rune Masked MaxGrid}
  n75 = Multiply(A=ScalarParameter() {defaultValue=2, group=Runes, parameterName=Rune Masked Pattern Tile}, B=TextureCoordinate())
  n76 = NamedRerouteUsage()
  n77 = ScalarParameter() {defaultValue=0.0135, group=Runes, parameterName=Rune Masked Pattern Time Speed}
  n78 = Time()
  n79 = TextureSampleParameter2D(UVs=Panner(Coordinate=Add(A=@n75, B=@n76), Time=Multiply(A=@n77, B=@n78)) {speedX=1}) {
   group=Runes
   parameterName=RuneTex
   samplerSource=SSM_Wrap_WorldGroupSettings
   samplerType=SAMPLERTYPE_LinearColor
   texture=/Game/Athena/Prototype/Blueprints/Cube/Materials/DFHeightMask
  }
  NamedRerouteDeclaration(None=Saturate(None=LinearInterpolate(A=@n73, B=@n74, Alpha=@n79.B))) {name=Rune Masked Pattern, nodeColor=<rgba 1 0.7266 0 1>}
  n80 = NamedRerouteUsage()
  n81 = ScalarParameter() {defaultValue=75, group=Runes, parameterName=Rune 4 Tile}
  n82 = MaterialFunctionCall(Float2=Add(A=Multiply(A=@n80, B=@n81), B=NamedRerouteUsage())) {materialFunction=&BreakOutFloat2Components}
  n83 = MaterialFunctionCall(
   X=Add(A=ScalarParameter() {group=Runes, parameterName=Rune 4 Offset U}, B=@n82.R)
   Y=Add(A=@n82.G, B=ScalarParameter() {defaultValue=0.5, group=Runes, parameterName=Rune 4 Offset V})
  ) {materialFunction=&MakeFloat2}
  n84 = TextureSampleParameter2D(UVs=@n83.Result) {
   group=Runes
   parameterName=RuneTex
   samplerSource=SSM_Wrap_WorldGroupSettings
   samplerType=SAMPLERTYPE_LinearColor
   texture=&DFHeightMask
  }
  n85 = LinearInterpolate(A=Constant() {r=-1}, B=Constant() {r=1}, Alpha=@n84.G) {constA=-1}
  n86 = ScalarParameter() {defaultValue=1, group=Runes, parameterName=Rune 4 Height Intensity}
  n87 = TextureSampleParameter2D!(
   UVs=BumpOffset(
    Coordinate=@n83.Result
    Height=Multiply(A=@n85, B=@n86)
    HeightRatioInput=ScalarParameter() {defaultValue=0.06, group=Runes, parameterName=Rune 4 HeightRatio}
   )
   Level=Constant()
   Apply View MipBias
  ) {
   group=Runes
   mipValueMode=TMVM_MipLevel
   parameterName=RuneTex
   samplerSource=SSM_Wrap_WorldGroupSettings
   samplerType=SAMPLERTYPE_LinearColor
   texture=&DFHeightMask
  }
  n88 = LinearInterpolate(
   A=ScalarParameter() {defaultValue=10, group=Runes, parameterName=Rune 4 Sine Max}
   B=ScalarParameter() {defaultValue=2, group=Runes, parameterName=Rune 4 Sine Min}
   Alpha=NamedRerouteUsage()
  )
  NamedRerouteDeclaration(None=LinearInterpolate(B=Power(Base=@n87.G, Exp=@n88), Alpha=ScalarParameter() {defaultValue=0.3, group=Runes, parameterName=Rune 4 Strength, sliderMax=1})) {name=Rune 04 Pattern, nodeColor=<rgba 1 0.7969 0 1>}
  n89 = ScalarParameter() {defaultValue=4, group=Runes, parameterName=Rune 03 Tile}
  n90 = NamedRerouteUsage()
  n91 = MaterialFunctionCall(Float2=Add(A=Multiply(A=@n89, B=@n90), B=NamedRerouteUsage())) {materialFunction=&BreakOutFloat2Components}
  n92 = MaterialFunctionCall(
   X=Add(A=ScalarParameter() {defaultValue=0.1, parameterName=Rune 03 Offset U}, B=@n91.R)
   Y=Add(A=@n91.G, B=ScalarParameter() {defaultValue=0.05, group=Runes, parameterName=Rune 03 Offset V})
  ) {materialFunction=&MakeFloat2}
  n93 = TextureSampleParameter2D!(UVs=@n92.Result, Apply View MipBias) {
   group=Runes
   parameterName=RuneTex
   samplerSource=SSM_Wrap_WorldGroupSettings
   samplerType=SAMPLERTYPE_LinearColor
   texture=&DFHeightMask
  }
  n94 = LinearInterpolate(A=Constant() {r=-1}, B=Constant() {r=1}, Alpha=@n93.G) {constA=-1}
  n95 = ScalarParameter() {defaultValue=0.5, group=Runes, parameterName=Rune 03 Height Intensity}
  n96 = TextureSampleParameter2D!(
   UVs=BumpOffset(
    Coordinate=@n92.Result
    Height=Multiply(A=@n94, B=@n95)
    HeightRatioInput=ScalarParameter() {defaultValue=0.15, group=Runes, parameterName=Rune 03 HeightRatio}
   )
   Level=Constant()
   Apply View MipBias
  ) {
   group=Runes
   mipValueMode=TMVM_MipLevel
   parameterName=RuneTex
   samplerSource=SSM_Wrap_WorldGroupSettings
   samplerType=SAMPLERTYPE_LinearColor
   texture=&DFHeightMask
  }
  NamedRerouteDeclaration(None=LinearInterpolate(B=@n96.G, Alpha=ScalarParameter() {defaultValue=1, group=Runes, parameterName=Rune 03 Strength, sliderMax=1})) {name=Rune 03 Pattern, nodeColor=<rgba 0.5469 0 1 1>}
  n97 = Add(A=Multiply(A=ScalarParameter() {defaultValue=32, group=Runes, parameterName=Rune 01 Tile}, B=TextureCoordinate()), B=NamedRerouteUsage())
  n98 = ScalarParameter() {defaultValue=-1, group=Runes, parameterName=Rune 01 HeightMin}
  n99 = ScalarParameter() {defaultValue=1, group=Runes, parameterName=Rune 01 Height Max}
  n100 = TextureSampleParameter2D!(UVs=@n97, Apply View MipBias) {
   group=Runes
   parameterName=RuneTex
   samplerSource=SSM_Wrap_WorldGroupSettings
   samplerType=SAMPLERTYPE_LinearColor
   texture=&DFHeightMask
  }
  n101 = TextureSampleParameter2D(UVs=BumpOffset(
    Coordinate=@n97
    Height=LinearInterpolate(A=@n98, B=@n99, Alpha=@n100.R)
    HeightRatioInput=ScalarParameter() {defaultValue=0.06, group=Runes, parameterName=Rune 01 Height Ratio}
   )) {
   group=Runes
   parameterName=RuneTex
   samplerSource=SSM_Wrap_WorldGroupSettings
   samplerType=SAMPLERTYPE_LinearColor
   texture=&DFHeightMask
  }
  n102 = ScalarParameter() {defaultValue=1.5, group=Runes, parameterName=Rune 01 Sine Min}
  n103 = ScalarParameter() {defaultValue=15, group=Runes, parameterName=Rune 01 Sine Max}
  n104 = NamedRerouteUsage()
  n105 = Multiply(
   A=Power(Base=@n101.R, Exp=LinearInterpolate(A=@n102, B=@n103, Alpha=@n104))
   B=ScalarParameter() {defaultValue=5, group=Runes, parameterName=Rune 01 Intensity}
  )
  NamedRerouteDeclaration(None=@n105) {name=Rune 01 Pattern, nodeColor=<rgba 1 0 0.0234 1>}
  n106 = NamedRerouteUsage()
  n107 = Saturate(None=Multiply(A=Multiply(A=@n101.R, B=@n106), B=ScalarParameter() {defaultValue=1.25, parameterName=Rune 01 Masked Intensity}))
  NamedRerouteDeclaration(None=@n107) {name=Rune 01 Pattern Masked, nodeColor=<rgba 0.2448 0.1912 0.1959 1>}
  n108 = ScalarParameter() {defaultValue=-0.1, group=Runes, parameterName=deactivated}
  n109 = ScalarParameter() {defaultValue=1, group=Runes, parameterName=activated}
  n110 = NamedRerouteUsage()
  n111 = VectorParameter() {group=Runes, parameterName=Rune Glow Color, defaultValue=<rgba 0.6076 0 1 1>}
  n112 = NamedRerouteUsage()
  n113 = ScalarParameter() {defaultValue=1, group=Runes, parameterName=Rune Outer PatternAlpha}
  n114 = LinearInterpolate(
   A=NamedRerouteUsage()
   B=Power(Base=NamedRerouteUsage(), Exp=ScalarParameter() {defaultValue=20, group=Runes, parameterName=Rune 3 Power})
   Alpha=Saturate(None=Multiply(A=@n112, B=@n113))
  )
  NamedRerouteDeclaration(None=Multiply(A=LinearInterpolate(A=@n108, B=@n109, Alpha=@n110), B=Multiply(A=@n111.RGB, B=@n114))) {name=Combined Rune Pattern, nodeColor=<rgba 1 0.4453 0 1>}
  NamedRerouteDeclaration(None=Max(A=@n3, B=@n4)) {name=NearMaskEmissive, nodeColor=<rgba 0 0.3438 1 1>}
  n115 = MaterialFunctionCall!(Material, Alpha) {materialFunction=/BRCosmetics/Characters/Player/Male/Medium/Bodies/M_MED_SpacePlunge/Materials/MF_StormNoise_Smoke}
  NamedRerouteDeclaration(None=Desaturation(None=@n115.Result) {luminanceFactors=<rgba 1 1 1 1>}) {name=Storm Noise, nodeColor=<rgba 1 0 0.9844 1>}
  NamedRerouteDeclaration(None=MaterialFunctionCall!(WorldScale=ScalarParameter() {defaultValue=75, parameterName=Master World UV Scale}, TransformToLocalSpace=StaticBool() {value=true}) {materialFunction=/Game/Items/Traps/Materials/MF_Creative_WorldAlignedUVs}.Result) {name=World Aligned UV Coordinates, nodeColor=<rgba 0.4766 1 0 1>}
  n116 = Subtract(A=@n6, B=ScalarParameter() {parameterName=RippleTime0})
  n117 = ScalarParameter() {defaultValue=0.6, parameterName=RippleDuration0}
  n118 = NamedRerouteDeclaration(None=Saturate(None=Divide(A=@n116, B=@n117))) {name=RippleTime0, nodeColor=<rgba 1 0 0.6797 1>}
  n119 = Subtract(A=@n6, B=ScalarParameter() {parameterName=RippleTime1})
  n120 = ScalarParameter() {defaultValue=0.4, parameterName=RippleDuration1}
  n121 = NamedRerouteDeclaration(None=Saturate(None=Divide(A=@n119, B=@n120))) {name=RippleTime1, nodeColor=<rgba 0.8516 0 1 1>}
  n122 = ScalarParameter() {parameterName=RippleTime2}
  n123 = Saturate(None=Divide(A=Subtract(A=@n6, B=@n122), B=ScalarParameter() {defaultValue=0.55, parameterName=RippleDuration2}))
  NamedRerouteDeclaration(None=Add(A=Add(A=@n118, B=@n121), B=NamedRerouteDeclaration(None=@n123) {name=RippleTime2, nodeColor=<rgba 0.8516 0 1 1>})) {name=RippleActive, nodeColor=<rgba 1 0 0.6797 1>}
  n124 = StaticSwitchParameter(
   True=MaterialFunctionCall!(PreskinnedPOS, Scanline Amount=ScalarParameter() {defaultValue=50, group=Distortion, parameterName=ScanlineAmount}) {materialFunction=/BRCosmetics/Accessories/FORT_Backpacks/Backpack_M_MED_SpacePlunge/Material/MF_Scanlines}.Result
   False=Constant()
  ) {group=Cube_Additional_Control, parameterName=UseHuntGlitch}
  NamedRerouteDeclaration(None=Multiply(A=SmoothStep(Value=@n124) {constMax=0.5}, B=Constant2Vector() {g=0.2, r=0.2}) {constB=0.1}) {name=Distortion, nodeColor=<rgba 1 0 0.9375 1>}
  n125 = Subtract(A=ActorPositionWS(), B=WorldPosition().XYZ)
  n126 = Distance(A=NamedRerouteUsage(), B=ActorPositionWS())
  n127 = Add(
   A=ScalarParameter() {defaultValue=8000, parameterName=CullDistanceOffset}
   B=ScalarParameter() {defaultValue=10000, parameterName=NearPlayerRadius}
  )
  n128 = Floor(None=Saturate(None=Divide(A=@n126, B=@n127)))
  n129 = Constant()
  n130 = If(A=NamedRerouteUsage(), "A > B"=@n129, "A == B"=@n129, "A < B"=Constant() {r=1}) {constB=2.99}
  NamedRerouteDeclaration(None=LinearInterpolate(A=Multiply(A=@n125, B=@n128), Alpha=Saturate(None=@n130)) {constB=0}) {name=WPO Distance Scale, nodeColor=<rgba 0.8047 1 0 1>}
  n131 = TextureObjectParameter() {
   group=Noise
   parameterName=CubeNoiseTexture
   samplerType=SAMPLERTYPE_Masks
   texture=/BRCosmetics/Accessories/FORT_Backpacks/Backpack_M_MED_SpacePlunge/Texture/T_CubeNoise_Character
  }
  NamedRerouteDeclaration(None=@n131) {name=CubeNoise_Texture, nodeColor=<rgba 0.7109 0 1 1>}
  n132 = Time()
  n133 = ScalarParameter() {defaultValue=1, group=Noise, parameterName=Noise_MotionSpeed}
  n134 = Add(
   A=WorldPosition() {worldPositionShaderOffset=WPT_ExcludeAllShaderOffsets}.XYZ
   B=Multiply(
    A=Multiply(A=@n132, B=@n133)
    B=VectorParameter() {group=Noise, parameterName=Noise_MotionDirection, channelNames {a=(none), b=Z, g=Y, r=X}}.RGB
   )
  )
  n135 = ScalarParameter() {defaultValue=4, group=Noise, parameterName=Noise_MipBias}
  n136 = AppendVector(A=@n43, B=Constant())
  NamedRerouteDeclaration(None=ComponentMask(None=MaterialFunctionCall!(
     World Space Normal
     TextureObject=@n48
     TextureSize=@n41
     WorldPosition=@n134
     Export Float 4
     MipBias=@n135
     RotZAxis
     RotXAxis
     RotYAxis
     UV_Offset=@n136
     ProjectionTransitionContrast
    ) {materialFunction=&/MF_DragonCart_WorldAlignedTexture}.XYZ Texture) {r=true}) {name=CubeDefaultNoise, nodeColor=<rgba 0 1 0.7891 1>}
  n137 = MaterialFunctionCall(TextureObject=@n47, TextureSize=@n41, WorldPosition=@n134, MipBias=@n135, UV_Offset=@n136) {materialFunction=&MF_DragonCart_WorldAlignedTexture}
  NamedRerouteDeclaration(None=ComponentMask(None=@n137.XYZ Texture) {r=true}) {name=BlurredShadow, nodeColor=<rgba 1 0 0.9844 1>}
  NamedRerouteDeclaration(None=ComponentMask(None=@n137.XYZ Texture) {g=true}) {name=BlurredNoise, nodeColor=<rgba 0.4531 0 1 1>}
  n138 = Multiply(A=NamedRerouteUsage(), B=@n29.Lerp Alpha No Contrast)
  n139 = ScalarParameter() {defaultValue=2.5, group=Cube_OG, parameterName=gpow3_1}
  n140 = NamedRerouteUsage()
  n141 = Add(A=@n40, B=Multiply(A=Power(Base=@n138, Exp=@n139), B=Multiply(A=@n39, B=@n140)))
  n142 = NamedRerouteUsage()
  n143 = LinearInterpolate(
   A=Multiply(A=ScalarParameter() {defaultValue=0.18, parameterName=InnerGlowWhileOff}, B=@n141)
   B=Multiply(A=@n141, B=Constant() {r=1})
   Alpha=@n142
  )
  n144 = Add(A=Multiply(A=NamedRerouteUsage(), B=@n143), B=@n143)
  n145 = NamedRerouteUsage()
  n146 = NamedRerouteUsage()
  n147 = Add(A=@n35, B=Multiply(A=@n35, B=Multiply(A=@n146, B=@n145)))
  n148 = LinearInterpolate(
   A=Multiply(A=ScalarParameter() {defaultValue=0.1, group=ZZ - Platform Quality, parameterName=LowQ-InnerGlowWhileOff}, B=@n147)
   B=Multiply(A=@n147, B=ScalarParameter() {defaultValue=0.4, group=ZZ - Platform Quality, parameterName=LowQ-InnerGlowWhileOn})
   Alpha=@n142
  )
  NamedRerouteDeclaration(None=LinearInterpolate(
    A=VectorParameter() {group=Runes, parameterName=GlowColor2, defaultValue=<rgba 0.0356 0.001 0.2 1>}.RGB
    B=QualitySwitch(Default=@n144, Low=@n148, High=@n144)
    Alpha=@n66
   ) {constA=0.3}) {name=Cube Aesthetic Emissive, nodeColor=<rgba 1 0.8906 0 1>}
  n149 = ScalarParameter() {defaultValue=1, parameterName=RuneActive}
  NamedRerouteDeclaration(None=@n149) {name=RuneActive, nodeColor=<rgba 0 1 1 1>}
  n150 = ComponentMask(None=TextureCoordinate()) {g=true}
  n151 = Multiply(A=Multiply(A=@n150, B=OneMinus(None=@n150))) {constB=17}
  NamedRerouteDeclaration(None=Saturate(None=Power(Base=@n151) {constExponent=6})) {name=Sphere Pinch Mask, nodeColor=<rgba 0 0.0625 1 1>}
  n152 = Time()
  n153 = ScalarParameter() {defaultValue=0.5, group=Cube_OG, parameterName=PulseSpeed_3}
  NamedRerouteDeclaration(None=MaterialFunctionCall(Value=Multiply(A=@n152, B=@n153)) {materialFunction=&LinearSine}."Rounded Linear Sine ") {name=AnimatedSine, nodeColor=<rgba 0.1484 1 0 1>}
  NamedRerouteDeclaration(None=Multiply(A=@n105, B=@n107)) {name=SinePattern, nodeColor=<rgba 0.7812 0 1 1>}
  NamedRerouteDeclaration(None=SphereMask(A=TextureCoordinate(), B=Constant() {r=0.5}, Radius=ScalarParameter() {defaultValue=0.7, parameterName=Sphere Mask UV 01 Radius}) {attenuationRadius=0.7, hardnessPercent=30}) {name=Sphere Mask UV 01, nodeColor=<rgba 1 0 0.9844 1>}
  n154 = Multiply(A=ScalarParameter() {group=Cube_OG, parameterName=Zap}) {constB=2}
  n155 = Multiply(A=ScalarParameter() {group=Cube_OG, parameterName=Zap2}) {constB=1.5}
  NamedRerouteDeclaration(None=Add(A=Add(A=@n154, B=@n155), B=@n149)) {name=RuneActiveMod, nodeColor=<rgba 0 1 0.6953 1>}
  n156 = VectorParameter() {group=Cube_OG, parameterName=SparklesColor, defaultValue=<rgba 0.5857 0.2 1 1>}
  n157 = ScalarParameter() {defaultValue=1, group=Cube_OG, parameterName=StarsTileMod}
  n158 = TextureCoordinate()
  n159 = Multiply(A=ScalarParameter() {defaultValue=6, group=Cube_OG, parameterName=SparkleTiling01}, B=@n158)
  n160 = Constant() {r=-2}
  n161 = Constant() {r=0.02}
  n162 = NamedRerouteUsage()
  n163 = TextureSample(UVs=Multiply(A=@n157, B=BumpOffset(Coordinate=@n159, Height=@n160, HeightRatioInput=@n161)), Tex=@n162) {samplerSource=SSM_Wrap_WorldGroupSettings, samplerType=SAMPLERTYPE_Masks, texture=&T_CubeNoise}
  n164 = Multiply(A=@n158, B=ScalarParameter() {defaultValue=12, parameterName=SparkleTiling02})
  n165 = Constant() {r=0.2}
  n166 = TextureSample(UVs=Multiply(A=@n157, B=BumpOffset(Coordinate=@n164, Height=@n160, HeightRatioInput=@n165)), Tex=@n162) {samplerSource=SSM_Wrap_WorldGroupSettings, samplerType=SAMPLERTYPE_Masks, texture=&T_CubeNoise}
  n167 = Multiply(A=@n163.B, B=@n166.B)
  n168 = ScalarParameter() {defaultValue=400, group=Cube_OG, parameterName=StarEmissiveRange02}
  n169 = Constant()
  n170 = LinearInterpolate(
   A=Constant() {r=7}
   B=Constant() {r=8}
   Alpha=NamedRerouteDeclaration(None=MaterialFunctionCall!(Blend Range=@n168, Start Offset=@n169) {materialFunction=/Engine/Functions/Engine_MaterialFunctions03/Particles/Distance_Blend}.Result) {name=DistanceLerp, nodeColor=<rgba 0 0.25 1 1>}
  )
  n171 = NamedRerouteDeclaration(None=Multiply(A=Power(Base=@n167, Exp=@n170), B=ScalarParameter() {defaultValue=40, group=Cube_OG, parameterName=SparkleIntensity})) {name=Sparkles, nodeColor=<rgba 0 0.5078 1 1>}
  NamedRerouteDeclaration(None=Multiply(A=Multiply(A=@n156.RGB, B=@n171), B=Saturate(None=@n5))) {name=Sparkle Emissive, nodeColor=<rgba 1 0.4453 0 1>}
  NamedRerouteDeclaration() {nodeColor=<rgba 1 0.1406 0 1>}
  n172 = MaterialFunctionCall!(
   RipplePosition=VectorParameter() {group=Ripple Anim, parameterName=RipplePosition0, defaultValue=<rgba -8743.291 -2119.8438 5633.1729 1>}.RGB
   RippleRadius=ScalarParameter() {defaultValue=2000, parameterName=RippleRadius0}
   RippleTime=NamedRerouteUsage()
  ) {materialFunction=&/MF_DragonCart_Orb_Barrier_Ripple}
  n173 = MaterialFunctionCall(
   RipplePosition=VectorParameter() {parameterName=RipplePosition1, defaultValue=<rgba -10197.0039 -2131.4111 5178.4331 0>}.RGB
   RippleRadius=ScalarParameter() {defaultValue=2000, parameterName=RippleRadius1}
   RippleTime=NamedRerouteUsage()
  ) {materialFunction=&MF_DragonCart_Orb_Barrier_Ripple}
  n174 = MaterialFunctionCall(
   RipplePosition=VectorParameter() {parameterName=RipplePosition2, defaultValue=<rgba -9644.1758 -599.1428 7150.7388 0>}.RGB
   RippleRadius=ScalarParameter() {defaultValue=2000, parameterName=RippleRadius2}
   RippleTime=NamedRerouteUsage()
  ) {materialFunction=&MF_DragonCart_Orb_Barrier_Ripple}
  NamedRerouteDeclaration(None=Max(A=Max(A=@n172.Sphere, B=@n173.Sphere), B=@n174.Sphere)) {name=RippleSphere, nodeColor=<rgba 1 0 0.5625 1>}
  n175 = Max(A=@n172.Ripple, B=@n173.Ripple)
  NamedRerouteDeclaration(None=Multiply(A=Max(A=@n175, B=@n174.Ripple)) {constB=1.5}) {name=Ripple, nodeColor=<rgba 1 0 0.5625 1>}
  n176 = Add(A=@n172.UVDistortion, B=@n173.UVDistortion)
  NamedRerouteDeclaration(None=Multiply(A=Add(A=@n176, B=@n174.UVDistortion), B=ScalarParameter() {defaultValue=-2, parameterName=RippleUVDistort})) {name=RippleUV, nodeColor=<rgba 1 0 0.5625 1>}
  NamedRerouteUsage()
  NamedRerouteUsage()
  NamedRerouteUsage()
  NamedRerouteUsage()
  NamedRerouteUsage()
  Reroute(None=@n8.RGB)
  Saturate(None=@n115.Result)
  ScalarParameter() {defaultValue=0.1}
  ScalarParameter() {defaultValue=1, group=Runes, parameterName=Test Rune Scale}
  TextureCoordinate() {vTiling=0.75}
  TextureCoordinate() {uTiling=0.95, vTiling=0.75}
 ]
}
