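# Translucent unlit test material for an impact ripple, filed with DragonCart's OrbBarrier VFX.
# A world-space sphere mask around Ripple Position, shaped by power, fade and a sine multiply, forms the strength-scaled "Impact Wave".
# The wave feeds a V-coordinate gradient lerp and a multiply with the borrowed MF_StormNoise_Smoke function.
# The graph looks unfinished: two color constants are unwired and no material output connections appear.
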Material /DragonCartNarrative/Assets/VFX/OrbBarrier/M_DragonCart_RippleTest
graph {
 properties {bAutomaticallySetUsageInEditor=false, blendMode=BLEND_Translucent, shadingModel=MSM_Unlit, usedShadingModels=Unlit}
 nodes [
  Constant3Vector() {constant=<rgba 1 0.1855 0.061 1>}
  Constant3Vector() {constant=<rgba 0.1699 0.9905 1 1>}
  n1 = Power(
   Base=SphereMask(
    A=WorldPosition().XYZ
    B=VectorParameter() {group=Ripple Anim, parameterName=Ripple Position, defaultValue=<rgba 0 0 0 1>}.RGB
    Radius=ScalarParameter() {defaultValue=3000, group=Ripple Anim, parameterName=RippleRadius}
   ) {hardnessPercent=0.5}
   Exp=ScalarParameter() {defaultValue=2, group=Ripple Anim, parameterName=Ripple Power}
  )
  n2 = Saturate(None=Multiply(A=@n1, B=ScalarParameter() {defaultValue=1, group=Ripple Anim, parameterName=Ripple Fade}))
  n3 = Saturate(None=Multiply(A=@n1, B=Sine(None=@n2)))
  n4 = NamedRerouteDeclaration(None=Multiply(A=LinearInterpolate(Alpha=@n3), B=ScalarParameter() {defaultValue=1, group=Ripple Anim, parameterName=Ripple Str})) {name=Impact Wave, nodeColor=<rgba 0 1 0.1562 1>}
  n5 = Multiply(A=ComponentMask(None=TextureCoordinate()) {g=true}) {constB=10}
  LinearInterpolate(B=@n4, Alpha=Saturate(None=Power(Base=@n5)))
  Multiply(
   A=MaterialFunctionCall!(Material, Alpha) {materialFunction=/BRCosmetics/Characters/Player/Male/Medium/Bodies/M_MED_SpacePlunge/Materials/MF_StormNoise_Smoke}.Result
   B=@n4
  )
 ]
}
